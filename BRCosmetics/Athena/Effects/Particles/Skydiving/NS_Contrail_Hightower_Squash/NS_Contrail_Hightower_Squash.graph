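# Skydiving contrail Niagara system for the Hightower Squash cosmetic, combining ribbon trails with white-blue lightning and wind around the character.
# Four skeleton-attached ribbons drift via drag and scaled inverse owner velocity, with separate gameplay and front-end speeds.
# Glow, spark, decal-style sprites, dive-reactive wind meshes and random lightning-mesh bursts with flashes and lights follow two sampled bones, likely the hands.
# Lit bone-flow sparks travel along a second skeletal mesh input.
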NiagaraSystem /BRCosmetics/Athena/Effects/Particles/Skydiving/NS_Contrail_Hightower_Squash/NS_Contrail_Hightower_Squash
dependencies {
 usedDataInterfaces [&NiagaraDataInterfaceCurve, &NiagaraDataInterfaceColorCurve]
 usedDynamicInputs []
 usedModules [
  &SystemState
  &self:SystemSpawnScript.SystemScriptSource.SystemScriptGraph.NiagaraNodeAssignment_0.SetVariables_#1
 ]
}
emitters {
 AirGlow {
  emitterUpdateScript [
   &EmitterState(
    Inactive Response=::NewEnumerator1
    Life Cycle Mode=::NewEnumerator1
    Loop Behavior=::NewEnumerator1
    Loop Duration Mode=::NewEnumerator1
    UseLoopDelay=true
   )
   &SpawnBurst_Instantaneous(Spawn Count=2, Use Spawn Probability=false)
  ]
  particleSpawnScript [
   /Niagara/Modules/Spawn/Initialization/InitializeParticle(
    Color=<rgba 1 1 1 0.04>
    Module.Write Color=true
    Module.Write Lifetime=true
    Module.Write Mass=false
    Module.Write Material Random=true
    Module.Write Position=false
    Module.Write Scale=false
    Module.Write SpriteRotation=false
    Module.Write SpriteSize=true
    Position=&SimulationPosition()
    Sprite Size=<24 24>
   )
   &SystemLocation(MaskSpawnGroup=false, Offset=<0 0 5>)
   &DynamicMaterialParameters(
    Index 0 Param 1=0.5
    Index 0 Param 2=50
    Module.Param0WriteEnabled=true
    Module.Param1WriteEnabled=true
    Module.Param2WriteEnabled=true
    Module.Param3WriteEnabled=true
   )
   &FN_SetSpawnGroup_v2(Number of Spawn Groups=2)
  ]
  particleUpdateScript [
   &ParticleState(Kill Particles When Lifetime Has Elapsed=false, Let Infinitely Lived Particles Die When Emitter Deactivates=true)
   &SkeletalMeshLocation(
    Bone / Socket Index=$Particles.SpawnGroup
    Bone Sampling Mode=::NewEnumerator4
    Inherit Velocity=false
    MaskSpawnGroup=false
    Sampled Position Offset=&MaskVectorBySpawnGroup(Masked Vector=<10 0 0>, Passthrough Vector=<-10 0 0>)
    Sampled Position Offset Coordinate Space:ENiagaraMeshTransforms=::NewEnumerator3
    Skeletal Mesh=$User.SkeletalMesh
   )
   &SolveForcesAndVelocity(Clamp Velocity=false, Limit Acceleration=false)
  ]
  rendererClasses [&NiagaraSpriteRendererProperties]
  renderers [{rendererClass=&NiagaraSpriteRendererProperties}]
 }
 Basic_Contrail {
  emitterUpdateScript [
   &EmitterState()
   &FN_SpawnRibbon_v2(Number of Ribbons=4, Spawn Mode=::NewEnumerator0, SpawnRate=40)
  ]
  particleSpawnScript [
   &FN_InitializeRibbon_v3(
    Enable Skeletal Position=true
    Lifetime=0.5625
    Ribbon Width=11
    Sampled Position Coordinate Space:ENiagaraCoordinateSpace=::Local
    Sampled Rotation Coordinate Space:ENiagaraCoordinateSpace=::Local
    Sampled Velocity Coordinate Space:ENiagaraCoordinateSpace=::Local
   ) {module=FN_InitializeRibbon_v3001}
   &AddVelocity(Velocity=&MakeCustomVectorFromBool(
     Bool=$Particles.Locked
     True Vector=<0 0 0>
     False Vector=&FN_FrontEndVector3Modifier(Gameplay=<0 0 400>, Front End=<0 0 350>)
    )) {module=AddVelocity001}
   &DynamicMaterialParameters(
    Index 0 Param 2=0.1
    Index 0 Param 3=1
    Index 0 Param 4=1
    Module.Param0WriteEnabled=true
    Module.Param1WriteEnabled=true
    Module.Param2WriteEnabled=true
    Module.Param3WriteEnabled=true
   )
  ]
  particleUpdateScript [
   &Drag(Drag=0.9, Use Linear Drag=true, Use Rotational Drag=false)
   &InheritVelocity(
    Module.Limit Speed=true
    Velocity Limit=10000
    Velocity Scale=&MakeCustomVectorFromBool(Bool=$Particles.Initial.Locked, True Vector=<0 0 0>, False Vector=<-0.03 -0.03 -0.01>)
    Velocity Source=&TransformVector(Vector=$Engine.Owner.Velocity, Source Space=::World, Destination Space=::Local)
   )
   &SolveForcesAndVelocity(Clamp Velocity=false, Limit Acceleration=false)
   &FN_UpdateRibbon(
    Enable Skeletal Position=true
    Initial Color:LinearColor=$Particles.Initial.Color
    Initial RibbonWidth:NiagaraFloat=$Particles.Initial.RibbonWidth
    Sampled Position Coordinate Space:ENiagaraCoordinateSpace=::Local
    Sampled Rotation Coordinate Space:ENiagaraCoordinateSpace=::Local
    Sampled Velocity Coordinate Space:ENiagaraCoordinateSpace=::Local
   )
   &ScaleRibbonWidth(Ribbon Width Scale=&FloatFromCurve(FloatCurve=DataInterface() {propertyValues {Curve {keys [
         {arriveTangent=1.9737, interpMode=RCIM_Cubic, leaveTangent=1.9737, tangentMode=RCTM_User, time=0.0058, value=0.9363}
         {arriveTangent=-0.0131, interpMode=RCIM_Cubic, leaveTangent=-0.0131, time=0.4538, value=1.5093}
         <at 0.6635 0.9277>
        ]}}}))
   &Color(
    Color=&ColorFromCurve(ColorCurve=DataInterface() {propertyValues {
       AlphaCurve {keys [<at 0.3 1 RCIM_Cubic>,<at 1 0>]}
       BlueCurve {keys [<at 0 1 RCIM_Cubic>,<at 1 1 RCIM_Cubic>]}
       GreenCurve {keys [<at 0 1 RCIM_Cubic>,<at 1 1 RCIM_Cubic>]}
       RedCurve {keys [<at 0 1 RCIM_Cubic>,<at 1 1 RCIM_Cubic>]}
      }})
    Scale Alpha=&FloatFromCurve(
     FloatCurve=DataInterface() {propertyValues {Curve {keys [
         {arriveTangent=2.7939, interpMode=RCIM_Cubic, leaveTangent=2.7939, tangentMode=RCTM_User, time=-0.0046, value=1.0113}
         {arriveTangent=-1.0036, interpMode=RCIM_Cubic, leaveTangent=-4.3444, tangentMode=RCTM_Break, time=0.1387, value=1.232}
         <at 0.6 0.002>
        ]}}}
     Scale Curve=0.8
    )
   )
  ]
  rendererClasses [&NiagaraRibbonRendererProperties]
  renderers [{rendererClass=&NiagaraRibbonRendererProperties}]
 }
 FN_BoneFlow_v2 {
  emitterUpdateScript [
   &EmitterState(
    Inactive Response=::NewEnumerator1
    Life Cycle Mode=::NewEnumerator1
    Loop Duration=&UniformRangedFloat(Minimum=0.05, Maximum=0.1)
    Recalculate Duration Each Loop=true
    UseLoopDelay=false
   )
   /Game/Effects/Niagara/Modules/Partition/FN_PartitionAttributes/FN_PartitionAttributes_v2(Partitions=2)
   &SpawnRate(SpawnRate=&UniformRangedFloat(Minimum=0, Maximum=10), Use Spawn Probability=false)
  ]
  particleSpawnScript [
   /Game/Effects/Niagara/Modules/Partition/FN_ParticlePartition_v2/FN_ParticlePartition_v2()
   &InitializeParticle(
    Color=<rgba 6 6 6 1>
    Module.Write Color=true
    Module.Write Lifetime=true
    Module.Write Mass=true
    Module.Write Material Random=true
    Module.Write Position=false
    Module.Write Scale=false
    Module.Write SpriteRotation=true
    Module.Write SpriteSize=true
    Sprite Rotation=&UniformRangedFloat(Minimum=0, Maximum=360)
    Sprite Size=&Vector2DFromFloat(Value=&UniformRangedFloat(Minimum=10, Maximum=30))
   )
   &DynamicMaterialParameters(
    Index 0 Param 3=&FloatFromCurve(FloatCurve=DataInterface() {propertyValues {Curve {keys [<at 0 1 RCIM_Cubic>,<at 0.2 0.025 RCIM_Cubic>,<at 1 1>]}}})
    Module.Param0WriteEnabled=false
    Module.Param1WriteEnabled=false
    Module.Param2WriteEnabled=true
    Module.Param3WriteEnabled=false
   )
   /Game/Effects/Niagara/Modules/BoneFlow/FN_Initialize_BoneFlow_Attributes_v2(
    Angle:NiagaraFloat=&UniformRangedFloat(Minimum=0, Maximum=360)
    Radius:NiagaraFloat=&UniformRangedFloat(Minimum=3, Maximum=10)
   )
   /SetVariables(Particles.SubImageIndex=&UniformRangedFloat(Minimum=0, Maximum=3)) {isSetParameters=true}
   /SetVariables(
    Particles.RandomRange:NiagaraFloat=&UniformRangedFloat(Minimum=0, Maximum=1)
    Particles.RandomRange2:NiagaraFloat=&UniformRangedFloat(Minimum=0.2, Maximum=1)
   ) {isSetParameters=true}
  ]
  particleUpdateScript [
   &ParticleState(Kill Particles When Lifetime Has Elapsed=false, Let Infinitely Lived Particles Die When Emitter Deactivates=true)
   /Game/Effects/Niagara/Modules/BoneFlow/FN_BoneFlow_v2/FN_BoneFlow_v2(
    Angle:NiagaraFloat=$Particles.FN_Initialize_BoneFlow_Attributes_v2.Angle
    Radius:NiagaraFloat=$Particles.FN_Initialize_BoneFlow_Attributes_v2.Radius
    SkeletalMesh:NiagaraDataInterfaceSkeletalMesh=$User.SkeletalMesh2
    T:NiagaraFloat=&Lerp_Float(
     A=$Particles.RandomRange
     B=&Add_Float(A=$Particles.RandomRange, B=$Particles.RandomRange2)
     Alpha=$Particles.NormalizedAge
    )
   )
   &DynamicMaterialParameters(
    Index 0 Param 3=&FloatFromCurve(FloatCurve=DataInterface() {propertyValues {Curve {keys [<at 0 1 RCIM_Cubic>,<at 0.2 0.025 RCIM_Cubic>,<at 1 1>]}}})
    Module.Param0WriteEnabled=false
    Module.Param1WriteEnabled=false
    Module.Param2WriteEnabled=true
    Module.Param3WriteEnabled=false
   ) {module=DynamicMaterialParameters001}
   &ScaleColor(
    Scale Alpha=&FloatFromCurve(FloatCurve=DataInterface() {propertyValues {Curve {keys [<at 0.75 1 RCIM_Cubic>, {arriveTangent=-10.3917, interpMode=RCIM_Cubic, leaveTangent=-10.3917, tangentMode=RCTM_User, time=1, value=0}]}}})
    Scale RGB=&Multiply_VectorByFloat(
     Vector=&MakeVectorFromLinearColorRGB(Color=&ColorFromCurve(ColorCurve=DataInterface() {propertyValues {
         AlphaCurve {keys []}
         BlueCurve {keys [<at 0 1>,<at 0.15 4.5 RCIM_Cubic>,<at 1 1 RCIM_Cubic>]}
         GreenCurve {keys [<at 0 1>, {arriveTangent=-0.1823, interpMode=RCIM_Cubic, leaveTangent=-0.1823, time=0.15, value=4.0703}, <at 1 0.8177 RCIM_Cubic>]}
         RedCurve {keys [<at 0 1>, {arriveTangent=-0.1812, interpMode=RCIM_Cubic, leaveTangent=-0.1812, time=0.15, value=3.8438}, <at 1 0.8188 RCIM_Cubic>]}
        }}))
     Float=1
    )
    ScaleA=true
    ScaleRGB=true
   )
   &CameraOffset(Camera Offset Amount=&UniformRangedFloat(Minimum=-10, Maximum=10))
   /Niagara/Modules/Update/SubUV/SubUVAnimation(Number Of Frames=8, Random SubUV Lookup Index:NiagaraFloat=$Particles.NormalizedAge, SubUV Animation Mode=::NewEnumerator1)
   &FN_Light_v3(
    Color=<rgba 10 10 10 1>
    Create Color Attribute=true
    Create Exponent Attribute=true
    Create Radius Attribute=true
    Enable Particle Light Enable=true
    Radius=25
   )
  ]
  rendererClasses [&NiagaraSpriteRendererProperties, &NiagaraLightRendererProperties]
  renderers [
   {rendererClass=&NiagaraSpriteRendererProperties}
   {rendererClass=&NiagaraLightRendererProperties, rendererIndex=1}
  ]
 }
 LightningFloating {
  emitterUpdateScript [
   &EmitterState(
    Inactive Response=::NewEnumerator1
    Life Cycle Mode=::NewEnumerator1
    Loop Duration=&UniformRangedFloat(Minimum=0.05, Maximum=0.1)
    UseLoopDelay=false
   )
   &SpawnRate(SpawnRate=&UniformRangedFloat(Minimum=0, Maximum=10), Use Spawn Probability=false)
  ]
  particleSpawnScript [
   &InitializeParticle(
    Module.Write Color=true
    Module.Write Lifetime=true
    Module.Write Mass=true
    Module.Write Material Random=true
    Module.Write Position=false
    Module.Write Scale=false
    Module.Write SpriteRotation=true
    Module.Write SpriteSize=true
    Position=&SimulationPosition()
    Sprite Rotation=&UniformRangedFloat(Minimum=0, Maximum=360)
    Sprite Size=&UniformRangedVector2D(Minimum=<10 10>, Maximum=<20 20>)
   )
   &DynamicMaterialParameters(
    Index 0 Param 3=&FloatFromCurve(FloatCurve=DataInterface() {propertyValues {Curve {keys [<at 0 1 RCIM_Cubic>,<at 0.2 0.025 RCIM_Cubic>,<at 1 1>]}}})
    Module.Param0WriteEnabled=false
    Module.Param1WriteEnabled=false
    Module.Param2WriteEnabled=true
    Module.Param3WriteEnabled=false
   )
   &FN_SetSpawnGroup_v2(Number of Spawn Groups=2)
   /SetVariables(
    Particles.RandomVector:Vector3f=&UniformRangedVector(Minimum=<-10 -10 0>, Maximum=<10 10 50>)
    Particles.RandomVector2:Vector3f=&UniformRangedVector(Minimum=<-10 -10 20>, Maximum=<10 10 40>)
   ) {isSetParameters=true}
  ]
  particleUpdateScript [
   &ParticleState()
   &SkeletalMeshLocation(
    Bone / Socket Index=$Particles.SpawnGroup
    Bone Sampling Mode=::NewEnumerator4
    Inherit Velocity=false
    MaskSpawnGroup=false
    Sampled Position Offset=&Lerp_Vector(
     A=$Particles.RandomVector
     B=&Add_Vector(A=$Particles.RandomVector, B=$Particles.RandomVector2)
     Alpha=$Particles.NormalizedAge
    )
    Skeletal Mesh=$User.SkeletalMesh
   )
   &SolveForcesAndVelocity(Clamp Velocity=false, Limit Acceleration=false)
   &DynamicMaterialParameters(
    Index 0 Param 3=&FloatFromCurve(FloatCurve=DataInterface() {propertyValues {Curve {keys [<at 0 1 RCIM_Cubic>,<at 0.2 0.025 RCIM_Cubic>,<at 1 1>]}}})
    Module.Param0WriteEnabled=false
    Module.Param1WriteEnabled=false
    Module.Param2WriteEnabled=true
    Module.Param3WriteEnabled=false
   ) {module=DynamicMaterialParameters001}
   &ScaleColor(
    Scale Alpha=&FloatFromCurve(FloatCurve=DataInterface() {propertyValues {Curve {keys [<at 0.75 1 RCIM_Cubic>, {arriveTangent=-10.3917, interpMode=RCIM_Cubic, leaveTangent=-10.3917, tangentMode=RCTM_User, time=1, value=0}]}}})
    Scale RGB=&Multiply_VectorByFloat(
     Vector=&MakeVectorFromLinearColorRGB(Color=&ColorFromCurve(ColorCurve=DataInterface() {propertyValues {
         AlphaCurve {keys []}
         BlueCurve {keys [<at 0 1>,<at 0.15 10 RCIM_Cubic>,<at 1 1 RCIM_Cubic>]}
         GreenCurve {keys [<at 0 1>, {arriveTangent=-0.1823, interpMode=RCIM_Cubic, leaveTangent=-0.1823, time=0.15, value=9.045}, <at 1 0.8177 RCIM_Cubic>]}
         RedCurve {keys [<at 0 1>, {arriveTangent=-0.1812, interpMode=RCIM_Cubic, leaveTangent=-0.1812, time=0.15, value=8.5417}, <at 1 0.8188 RCIM_Cubic>]}
        }}))
     Float=1
    )
    ScaleA=true
    ScaleRGB=true
   )
   &SubUVAnimation(Number Of Frames=8, Random SubUV Lookup Index:NiagaraFloat=$Particles.NormalizedAge, SubUV Animation Mode=::NewEnumerator1)
  ]
  rendererClasses [&NiagaraSpriteRendererProperties]
  renderers [{rendererClass=&NiagaraSpriteRendererProperties}]
 }
 LightningHands {
  emitterUpdateScript [
   &EmitterState()
   &SpawnRate(SpawnRate=&UniformRangedFloat(Minimum=0, Maximum=5), Use Spawn Probability=false)
  ]
  particleSpawnScript [
   &FN_SetSpawnGroup_v2(Number of Spawn Groups=2)
   &InitializeParticle(
    Color=&MakeLinearColorFromVectorAndFloat(
     "Vector (RGB)"=&MakeVectorFromLinearColorRGB(Color=<rgba 1.5 1.8928 2 1>)
     "Float (Alpha)"=&UniformRangedFloat(Minimum=0.25, Maximum=1)
    )
    Lifetime=&UniformRangedFloat(Minimum=0.45, Maximum=0.65)
    Module.Write Color=true
    Module.Write Lifetime=true
    Module.Write Mass=false
    Module.Write Material Random=true
    Module.Write Position=false
    Module.Write Scale=false
    Module.Write SpriteRotation=true
    Module.Write SpriteSize=true
    Sprite Rotation=&NormalizedAngleToDegrees(NormalizedAngle=&UniformRangedFloat(Minimum=0, Maximum=1))
    Sprite Size=&Vector2DFromFloat(Value=&UniformRangedFloat(Minimum=20, Maximum=32))
   )
   /SetVariables(Particles.SubImageIndex=&UniformRangedFloat(Minimum=0, Maximum=3)) {isSetParameters=true}
   &DynamicMaterialParameters(
    Index 0 Param 1=&UniformRangedFloat(Minimum=0, Maximum=0)
    Index 0 Param 2=&UniformRangedFloat(Minimum=1, Maximum=2)
    Module.Param0WriteEnabled=true
    Module.Param1WriteEnabled=true
    Module.Param2WriteEnabled=false
    Module.Param3WriteEnabled=false
   )
   /Niagara/Modules/Spawn/Location/SphereLocation(MaskSpawnGroup=false, Module.Override Randomness=false, Sphere Radius=5, Surface Only=false)
  ]
  particleUpdateScript [
   &ParticleState(Kill Particles When Lifetime Has Elapsed=false, Let Infinitely Lived Particles Die When Emitter Deactivates=true)
   &SkeletalMeshLocation(
    Bone / Socket Index=$Particles.SpawnGroup
    Bone Sampling Mode=::NewEnumerator4
    Inherit Velocity=false
    Interpolate Mesh Sampling=false
    MaskSpawnGroup=false
    Sampled Position Offset=&Add_Vector(
     A=Linked() {linkedVariable {name=Particles.Initial.Position, type=NiagaraPosition}}
     B=&MaskVectorBySpawnGroup(Masked Vector=<5 0 0>, Passthrough Vector=<-5 0 0>)
    )
    Sampled Position Offset Coordinate Space:ENiagaraMeshTransforms=::NewEnumerator3
    Skeletal Mesh=$User.SkeletalMesh
   ) {module=SkeletalMeshLocation001}
   &SolveForcesAndVelocity(Clamp Velocity=false, Limit Acceleration=false)
   &DynamicMaterialParameters(
    Index 0 Param 3=&FloatFromCurve(FloatCurve=DataInterface() {propertyValues {Curve {keys [<at 0 1 RCIM_Cubic>,<at 0.2 0 RCIM_Cubic>,<at 1 1>]}}})
    Module.Param0WriteEnabled=false
    Module.Param1WriteEnabled=false
    Module.Param2WriteEnabled=true
    Module.Param3WriteEnabled=false
   ) {module=DynamicMaterialParameters001}
   &ScaleColor(
    Scale Alpha=&FloatFromCurve(FloatCurve=DataInterface() {propertyValues {Curve {keys [<at 0.75 1 RCIM_Cubic>,<at 1 0>]}}})
    Scale RGB=&MakeVectorFromLinearColorRGB(Color=&ColorFromCurve(ColorCurve=DataInterface() {propertyValues {
        AlphaCurve {keys []}
        BlueCurve {keys [<at 0 3 RCIM_Cubic>,<at 1 1 RCIM_Cubic>]}
        GreenCurve {keys [<at 0 2.7135 RCIM_Cubic>,<at 1 0.8177 RCIM_Cubic>]}
        RedCurve {keys [<at 0 2.5625 RCIM_Cubic>,<at 1 0.8188 RCIM_Cubic>]}
       }}))
    ScaleA=true
    ScaleRGB=true
   )
  ]
  rendererClasses [&NiagaraSpriteRendererProperties]
  renderers [{rendererClass=&NiagaraSpriteRendererProperties}]
 }
 Lightning_Mesh_2 {
  emitterUpdateScript [
   &EmitterState(
    Inactive Response=::NewEnumerator1
    Life Cycle Mode=::NewEnumerator1
    Loop Delay=2
    Loop Duration=&UniformRangedFloat(Minimum=0.1, Maximum=0.85, Randomness Mode=::NewEnumerator1)
    Recalculate Duration Each Loop=true
    UseLoopDelay=false
   )
   /SetVariables(Emitter.RandSpawnGroup:NiagaraInt32=&MakeIntFromFloat(Float=&UniformAOrBFloat(A=0, Randomness Mode=::NewEnumerator1, Random Seed=$Emitter.LoopCount, Fixed Random Seed=true))) {isSetParameters=true}
   &SpawnBurst_Instantaneous(Spawn Count=&UniformRangedInt(Minimum=1, Maximum=3), Spawn Probability=0.5, Use Spawn Probability=false)
  ]
  particleSpawnScript [
   /Game/Effects/Niagara/Modules/Math/FN_EvenNumberToSign(Input:NiagaraInt32=$Particles.SpawnGroup, Invert Sign=true, Output Float Attribute=true)
   &InitializeParticle(
    Color=<rgba 1.5 1.8928 2 1>
    Lifetime=&UniformRangedFloat(Minimum=0.2, Maximum=0.25)
    Mesh Scale=&UniformRangedVector(Minimum=<10 6 6>, Maximum=<20 10 10>)
    Module.Write Color=true
    Module.Write Lifetime=true
    Module.Write Mass=true
    Module.Write Material Random=true
    Module.Write Position=false
    Module.Write Scale=true
    Module.Write SpriteRotation=false
    Module.Write SpriteSize=false
    Position=&SimulationPosition()
   )
   &DynamicMaterialParameters(
    Index 0 Param 1=6
    Index 0 Param 3=&UniformRangedFloat(Minimum=1.45, Maximum=2.2)
    Index 0 Param 4=&UniformRangedFloat(Minimum=-60, Maximum=60)
    Module.Param0WriteEnabled=true
    Module.Param1WriteEnabled=false
    Module.Param2WriteEnabled=true
    Module.Param3WriteEnabled=true
   )
   /Niagara/Modules/Spawn/Orientation/V2/InitialMeshOrientation(
    Rotation=&UniformRangedVector(Minimum=<-1 0 0>, Maximum=<1 0 0>)
    Rotation Coordinate Space:ENiagaraMeshTransforms=::NewEnumerator1
    Use Orientation Vector=true
    Use Rotation Vector=true
   )
   /SetVariables(
    Particles.FlashCameraOffset=20
    Particles.SpikeySprite_DP:Vector4f=&MakeVector4(X=&UniformRangedFloat(Minimum=0, Maximum=1), Y=0, Z=0, W=0)
   ) {isSetParameters=true}
   /SetVariables(Particles.MeshOrientation=&MultiplyQuaternion(
     Quaternion A=&FindQuatBetweenTwoVectors(From Vector=<1 0 0>, To Vector=&UniformRangedVector(Minimum=<-0.05 -0.05 1>, Maximum=<0.05 0.05 1>))
     Quaternion B=$Particles.Initial.MeshOrientation
    )) {isSetParameters=true}
   /SetVariables(
    Particles.LensFlare_PositionRange:Vector3f=&UniformRangedVector(Minimum=<-10 -10 0>, Maximum=<10 10 15>)
    Particles.LensFlare_Size:Vector2f=&Vector2DFromFloat(Value=&UniformRangedFloat(Minimum=5, Maximum=15))
    Particles.LensFlare_Visibility:NiagaraBool=/Niagara/DynamicInputs/Random/RandomBool(Randomness Mode:ENiagaraRandomnessMode=::NewEnumerator0, Random Seed:NiagaraInt32=0, Fixed Random Seed:NiagaraBool=false)
   ) {isSetParameters=true}
  ]
  particleUpdateScript [
   &ParticleState()
   &SkeletalMeshLocation(
    Bone / Socket Index=$Emitter.RandSpawnGroup
    Bone Sampling Mode=::NewEnumerator4
    Inherit Velocity=false
    Interpolate Mesh Sampling=false
    MaskSpawnGroup=false
    Sampled Position Offset=&Lerp_Vector(A=<5 0 0>, B=<-5 0 0>, Alpha=&MakeFloatFromInt(Int=$Emitter.RandSpawnGroup))
    Sampled Position Offset Coordinate Space:ENiagaraMeshTransforms=::NewEnumerator3
    Skeletal Mesh=$User.SkeletalMesh
   )
   &DynamicMaterialParameters(
    Index 0 Param 2=&FloatFromCurve(FloatCurve=DataInterface() {propertyValues {Curve {keys [{arriveTangent=-2.9283, interpMode=RCIM_Cubic, leaveTangent=-2.9283, tangentMode=RCTM_User, time=0.1, value=1}, <at 1 -1>]}}})
    Module.Param0WriteEnabled=false
    Module.Param1WriteEnabled=true
    Module.Param2WriteEnabled=false
    Module.Param3WriteEnabled=false
   ) {module=DynamicMaterialParameters001}
   /SetVariables(
    Particles.LensFlare_Color:LinearColor=&MakeCustomLinearColorFromBool(
     Bool=$Particles.LensFlare_Visibility
     True Color=&ColorFromCurve(ColorCurve=DataInterface() {propertyValues {
        AlphaCurve {keys [<at 0 0 RCIM_Cubic>,<at 0.25 1>,<at 0.6 1>,<at 0.8 0 RCIM_Cubic>]}
        BlueCurve {keys [<at 0 1 RCIM_Cubic>,<at 0.2 6>,<at 1 1 RCIM_Cubic>]}
        GreenCurve {keys [<at 0 1 RCIM_Cubic>,<at 0.2 6>,<at 1 1 RCIM_Cubic>]}
        RedCurve {keys [<at 0 1 RCIM_Cubic>,<at 0.2 6>,<at 1 1 RCIM_Cubic>]}
       }})
     False Color=<rgba 1 1 1 0>
    )
    Particles.LensFlare_DynamicParameter:Vector4f=&MakeVector4(
     X=&UniformRangedFloat(Minimum=0, Maximum=1)
     Y=&UniformRangedFloat(Minimum=1, Maximum=4)
     Z=1
     W=&FloatFromCurve(FloatCurve=DataInterface() {propertyValues {Curve {keys [<at 0 1>,<at 0.2 0>,<at 0.4625 0.6907>,<at 1 1>]}}})
    )
    Particles.LensFlare_Position:Vector3f=&Add_Vector(A=Linked() {linkedVariable {name=Particles.Position, type=NiagaraPosition}}, B=$Particles.LensFlare_PositionRange)
   ) {isSetParameters=true}
   /SetVariables(
    Particles.FlashColor:LinearColor=&ColorFromCurve(ColorCurve=DataInterface() {propertyValues {
       AlphaCurve {keys [
         <at 0.0466 -0.0027>
         <at 0.1143 0.9973>
         {arriveTangent=-3.3109, interpMode=RCIM_Cubic, leaveTangent=-3.3109, time=0.2215, value=0.9973}
         <at 0.4163 -0.0027 RCIM_Cubic>
        ]}
       BlueCurve {keys [<at 0.2763 7.9973 RCIM_Cubic>,<at 1 2 RCIM_Cubic>]}
       GreenCurve {keys [<at 0.2763 6.6774 RCIM_Cubic>,<at 1 1.8793 RCIM_Cubic>]}
       RedCurve {keys [<at 0.2763 2.5807 RCIM_Cubic>,<at 1 1.5048 RCIM_Cubic>]}
      }})
    Particles.FlashPosition:NiagaraPosition=$Particles.Position
    Particles.FlashScale:Vector2f=<40 40>
   ) {isSetParameters=true}
   &ScaleColor(
    Scale RGB=&MakeVectorFromLinearColorRGB(Color=&ColorFromCurve(ColorCurve=DataInterface() {propertyValues {
        AlphaCurve {keys []}
        BlueCurve {keys [<at 0 2 RCIM_Cubic>,<at 1 4 RCIM_Cubic>]}
        GreenCurve {keys [<at 0 1.8653 RCIM_Cubic>,<at 1 2.6922 RCIM_Cubic>]}
        RedCurve {keys [<at 0 1.5312 RCIM_Cubic>,<at 1 0.8958 RCIM_Cubic>]}
       }}))
    ScaleA=true
    ScaleRGB=true
   )
   &FN_Light_v3(
    Color=&ColorFromCurve(ColorCurve=DataInterface() {propertyValues {
       AlphaCurve {keys []}
       BlueCurve {keys [<at 0 1 RCIM_Cubic>,<at 0.126 5 RCIM_Cubic>,<at 0.4308 1 RCIM_Cubic>,<at 0.6953 5 RCIM_Cubic>,<at 1 1 RCIM_Cubic>]}
       GreenCurve {keys [
         <at 0 0.9237 RCIM_Cubic>
         <at 0.126 4.469 RCIM_Cubic>
         {arriveTangent=0.5872, interpMode=RCIM_Cubic, leaveTangent=0.5872, time=0.4308, value=0.9237}
         <at 0.6953 4.8033 RCIM_Cubic>
         <at 1 0.9237 RCIM_Cubic>
        ]}
       RedCurve {keys [
         <at 0 0.8 RCIM_Cubic>
         <at 0.126 3.6082 RCIM_Cubic>
         {arriveTangent=1.5393, interpMode=RCIM_Cubic, leaveTangent=1.5393, time=0.4308, value=0.8}
         <at 0.6953 4.4844 RCIM_Cubic>
         <at 1 0.8 RCIM_Cubic>
        ]}
      }})
    Create Color Attribute=true
    Create Position Attribute=true
    Create Radius Attribute=true
    Position:Vector3f=&Add_Vector(A=Linked() {linkedVariable {name=Particles.Position, type=NiagaraPosition}}, B=<0 0 60>)
    Radius:NiagaraFloat=&Multiply_Float(
     A=75
     B=&FloatFromCurve(FloatCurve=DataInterface() {propertyValues {Curve {keys [<at 0 0.25>,<at 0.1667 1.0351>,<at 0.2561 1.0244>,<at 0.4349 0.7686>,<at 0.6006 1.0138>,<at 0.6329 1.0031>,<at 0.8454 0.7457>]}}})
    )
   )
   /Game/Effects/Niagara/Modules/Mesh/FN_MeshRenderer_v3/FN_MeshRenderer_v2(
    Color:LinearColor=&MultiplyLinearColor(
     Linear Color=<rgba 0.5195 0.7193 0.75 0.1>
     Scale Factor=&MakeFloat4FromVectorAndFloat(
      "Vector (RGB)"=&MakeVectorFromLinearColorRGB(Color=&ColorFromCurve(ColorCurve=DataInterface() {propertyValues {
          AlphaCurve {keys []}
          BlueCurve {keys [<at 0 2 RCIM_Cubic>,<at 0.2 4 RCIM_Cubic>,<at 0.5 2 RCIM_Cubic>]}
          GreenCurve {keys [<at 0 1.8793 RCIM_Cubic>,<at 0.2 3.7587 RCIM_Cubic>,<at 0.5 1.8793 RCIM_Cubic>]}
          RedCurve {keys [<at 0 1.5048 RCIM_Cubic>,<at 0.2 3.0096 RCIM_Cubic>,<at 0.5 1.5048 RCIM_Cubic>]}
         }}))
      "Float (Alpha)"=&FloatFromCurve(FloatCurve=DataInterface() {propertyValues {Curve {keys [
           <at 0.1 1>
           {arriveTangent=-0.7039, interpMode=RCIM_Cubic, leaveTangent=-0.7039, time=0.3472, value=1}
           <at 0.5972 0.65 RCIM_Cubic>
           <at 0.8472 1>
           <at 1.0972 0.65 RCIM_Cubic>
           <at 1.3472 1 RCIM_Cubic>
           <at 1.5972 0.65>
           <at 1.8472 1 RCIM_Cubic>
          ]}}})
     )
    )
    Create COlor Attribute=true
    Create Scale Attribute=true
    Scale:Vector3f=<1 1 1>
   ) {module=FN_MeshRenderer_v2001}
  ]
  rendererClasses [
   &NiagaraMeshRendererProperties
   &NiagaraLightRendererProperties
   &NiagaraSpriteRendererProperties
  ]
  renderers [
   {rendererClass=&NiagaraMeshRendererProperties, rendererIndex=0}
   {rendererClass=&NiagaraMeshRendererProperties, rendererIndex=1}
   {rendererClass=&NiagaraLightRendererProperties, rendererIndex=2}
   {rendererClass=&NiagaraSpriteRendererProperties, rendererIndex=3}
   {rendererClass=&NiagaraSpriteRendererProperties, rendererIndex=4}
  ]
 }
 Particle_Decal002 {
  emitterUpdateScript [
   &EmitterState(Inactive Response=::NewEnumerator1, Life Cycle Mode=::NewEnumerator1, Loop Duration=1, UseLoopDelay=false)
   &SpawnRate(SpawnRate=20, Use Spawn Probability=false)
  ]
  particleSpawnScript [
   &InitializeParticle(
    Color=<rgba 6.9583 7.7462 8 0.5>
    Lifetime=0.5
    Module.Write Color=true
    Module.Write Lifetime=true
    Module.Write Mass=false
    Module.Write Material Random=true
    Module.Write Position=false
    Module.Write Scale=false
    Module.Write SpriteRotation=true
    Module.Write SpriteSize=true
    Position=&SimulationPosition()
    Sprite Rotation=&NormalizedAngleToDegrees(NormalizedAngle=&UniformRangedFloat(Minimum=-1, Maximum=1))
    Sprite Size=&Vector2DFromFloat(Value=&UniformRangedFloat(Minimum=30, Maximum=40))
   )
   &FN_SetSpawnGroup_v2(Number of Spawn Groups=2)
   &DynamicMaterialParameters(
    Index 0 Param 2=&UniformRangedFloat(Minimum=0, Maximum=10)
    Index 0 Param 3=1
    Index 0 Param 4=1
    Module.Param0WriteEnabled=false
    Module.Param1WriteEnabled=true
    Module.Param2WriteEnabled=true
    Module.Param3WriteEnabled=true
   )
   &CameraOffset(Camera Offset Amount=6)
  ]
  particleUpdateScript [
   &ParticleState()
   &DynamicMaterialParameters(
    Index 0 Param 1=&FloatFromCurve(FloatCurve=DataInterface() {propertyValues {Curve {keys [<at 0 0.01>,<at 1 1.5>]}}})
    Module.Param0WriteEnabled=true
    Module.Param1WriteEnabled=false
    Module.Param2WriteEnabled=false
    Module.Param3WriteEnabled=false
   ) {module=DynamicMaterialParameters001}
   &SkeletalMeshLocation(
    Bone / Socket Index=$Particles.SpawnGroup
    Bone Sampling Mode=::NewEnumerator4
    Inherit Velocity=false
    Interpolate Mesh Sampling=false
    MaskSpawnGroup=false
    Sampled Position Offset=&Multiply_VectorByFloat(
     Vector=&MaskVectorBySpawnGroup(Masked Vector=<9 -0.5 1.25>, Passthrough Vector=<-9 -0.5 1.25>)
     Float=$Particles.NormalizedAge
    )
    Sampled Position Offset Coordinate Space:ENiagaraMeshTransforms=::NewEnumerator3
    Skeletal Mesh=$User.SkeletalMesh
    Use Sampled Mesh Attribute as Ribbon ID=false
   )
   &ScaleColor(
    Scale Alpha=&FloatFromCurve(FloatCurve=DataInterface() {propertyValues {Curve {keys [<at 0 1>,<at 0.2 0>,<at 1 0.5>]}}})
    Scale RGB=&VectorFromFloat(Value=&FloatFromCurve(FloatCurve=DataInterface() {propertyValues {Curve {keys [<at 0 1>,<at 0.1 1.5>,<at 0.3094 1.2707>,<at 1 1>]}}}))
    ScaleA=true
    ScaleRGB=true
   )
   &ScaleSpriteSize(Scale Factor=&Vector2DFromFloat(Value=&FloatFromCurve(FloatCurve=DataInterface() {propertyValues {Curve {keys [<at 0 0.15>,<at 0.4522 0.5145>,<at 0.9 1>]}}})))
  ]
  rendererClasses [&NiagaraSpriteRendererProperties]
  renderers [{rendererClass=&NiagaraSpriteRendererProperties}]
 }
 WindHands {
  emitterUpdateScript [
   &EmitterState(Life Cycle Mode=::NewEnumerator1, Loop Duration=1, UseLoopDelay=false)
   &SpawnRate(SpawnRate=20, Use Spawn Probability=false)
   /Game/Effects/Niagara/Modules/Skydiving/FN_DivingAttributes/FN_EmitterDivingAttributes_v1(Bone Index=5, SkeletalMesh:NiagaraDataInterfaceSkeletalMesh=$User.SkeletalMesh)
  ]
  particleSpawnScript [
   &FN_SetSpawnGroup_v2(Number of Spawn Groups=2)
   &InitializeParticle(
    Color=<rgba 1 1 1 0.2>
    Lifetime=&UniformRangedFloat(Minimum=0.2, Maximum=0.4)
    Mesh Scale=<1 1 1.5>
    Module.Write Color=true
    Module.Write Lifetime=true
    Module.Write Mass=true
    Module.Write Material Random=true
    Module.Write Position=false
    Module.Write Scale=true
    Module.Write SpriteRotation=false
    Module.Write SpriteSize=false
    Position=&SimulationPosition()
   )
   &InitialMeshOrientation(
    Rotation=&UniformRangedVector(Minimum=<0 0 -1>, Maximum=<0 0 1>)
    Use Orientation Vector=true
    Use Rotation Vector=true
   )
  ]
  particleUpdateScript [
   &ParticleState()
   /BRCosmetics/Effects/Niagara/Modules/Orientation/FN_ReorientMesh()
   &SkeletalMeshLocation(
    Bone / Socket Index=$Particles.SpawnGroup
    Bone Sampling Mode=::NewEnumerator4
    Inherit Velocity=false
    Interpolate Mesh Sampling=false
    MaskSpawnGroup=false
    Sampled Position Offset=&Lerp_Vector(
     A=&MaskVectorBySpawnGroup(
      Masked Vector=&Lerp_Vector(A=<0 0 0>, B=<0 0 0>, Alpha=$Particles.NormalizedAge)
      Passthrough Vector=&Lerp_Vector(A=<0 0 0>, B=<0 0 0>, Alpha=$Particles.NormalizedAge)
     )
     B=&MaskVectorBySpawnGroup(
      Masked Vector=&Lerp_Vector(A=<0 0 0>, B=<20 0 0>, Alpha=$Particles.NormalizedAge)
      Passthrough Vector=&Lerp_Vector(A=<0 0 0>, B=<-20 0 0>, Alpha=$Particles.NormalizedAge)
     )
     Alpha=$Emitter.Dive
    )
    Sampled Position Offset Coordinate Space:ENiagaraMeshTransforms=::NewEnumerator3
    Skeletal Mesh=$User.SkeletalMesh
   )
   /SetVariables(Particles.MeshOrientation=&MultiplyQuaternion(
     Quaternion A=&MakeQuatFromAxisAngle(
      Axis=&RotateVectorByQuaternion(Vector To Rotate=<1 0 0>, Quaternion=$Output.SkeletalMeshLocation.SampledRotation)
      AngleInDegrees=&Multiply_Float(A=$Particles.NormalizedAge, B=&MaskFloatBySpawnGroup(Masked Float=75, Passthrough Float=-75))
     )
     Quaternion B=&MultiplyQuaternion(
      Quaternion A=$Output.SkeletalMeshLocation.SampledRotation
      Quaternion B=&MultiplyQuaternion(Quaternion A=$Particles.Orientation, Quaternion B=$Particles.Initial.MeshOrientation)
     )
    )) {isSetParameters=true}
   &ScaleMeshSize(Scale Factor=&VectorFromFloat(Value=&FloatFromCurve(FloatCurve=DataInterface() {propertyValues {Curve {keys [<at 0 1>,<at 1 2>]}}})))
   &ScaleColor(
    Scale Alpha=&FloatFromCurve(FloatCurve=DataInterface() {propertyValues {Curve {keys [<at 0 0 RCIM_Cubic>,<at 0.5 1 RCIM_Cubic>,<at 1 0 RCIM_Cubic>]}}})
    ScaleA=true
    ScaleRGB=true
   )
  ]
  rendererClasses [&NiagaraMeshRendererProperties]
  renderers [{rendererClass=&NiagaraMeshRendererProperties, rendererIndex=0}]
 }
}
summary {
 systemName=NS_Contrail_Hightower_Squash
 emitters [
  {
   emitterName=Lightning_Mesh_2
   rendererClasses [
    &NiagaraMeshRendererProperties
    &NiagaraLightRendererProperties
    &NiagaraSpriteRendererProperties
   ]
  }
  {emitterName=WindHands, rendererClasses [&NiagaraMeshRendererProperties]}
  {emitterName=AirGlow, rendererClasses [&NiagaraSpriteRendererProperties]}
  {emitterName=Basic_Contrail, rendererClasses [&NiagaraRibbonRendererProperties]}
  {emitterName=LightningHands, rendererClasses [&NiagaraSpriteRendererProperties]}
  {
   emitterName=FN_BoneFlow_v2
   rendererClasses [&NiagaraSpriteRendererProperties, &NiagaraLightRendererProperties]
  }
  {emitterName=LightningFloating, rendererClasses [&NiagaraSpriteRendererProperties]}
  {emitterName=Particle_Decal002, rendererClasses [&NiagaraSpriteRendererProperties]}
 ]
 userVariables [
  User.bIsFrontEnd {type=NiagaraBool}
  User.bIsFrontEndPreview {type=NiagaraBool}
  User.PlayerSpeed {type=NiagaraFloat, defaultValue=0, description=""}
  User.RibbonIDOffset {type=NiagaraInt32}
  User.SkeletalMesh {
   type=NiagaraDataInterfaceSkeletalMesh
   defaultValue {
    struct=NiagaraExt_VariableValue_DataInterface
    value {dataInterface=&self:NiagaraDataInterfaceSkeletalMesh_2}
   }
  }
  User.SkeletalMesh2 {
   type=NiagaraDataInterfaceSkeletalMesh
   defaultValue {
    struct=NiagaraExt_VariableValue_DataInterface
    value {dataInterface=&self:NiagaraDataInterfaceSkeletalMesh_1}
   }
  }
 ]
}
